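; Hanging cargo container prop for the Helios environment, built on a parent Blueprint with an adjustable light fixture component.
; Construction applies the first building texture data entry's override material and tint to the fixture, then hides the inherited static mesh component and adds a new one with the same mesh.
; Exposed asset provider overrides register version 2 and allow references from the main material, its Nanite override and the main mesh.
; The event graph stubs are empty.
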(event EventBeginPlay)

(event Collision|EventActorBeginOverlap
  (params (OtherActor :type Actor%20Object%20Reference)))

(event EventTick
  (params (DeltaSeconds :type Float%20%28single-precision%29)))

(fn ApplyCorrectMIDFromOverride
  (bind _lightfixture (Variables|AdjustableLight|GetLightFixture))
  (bind _outtexturedataarray (Building|GetBuildingTextureData))
  (if (> (Utilities|Array|Length _outtexturedataarray) 0)
    (Rendering|Material|CreateDynamicMaterialInstance _lightfixture 0 (Variables|/Script/FortniteGame.BuildingTextureData|GetOverrideMaterial (Utilities|Array|Get(acopy) _outtexturedataarray)) "None" :!class /Script/Engine.PrimitiveComponent)
    (Default|SetColoronSecondaryMesh (Math|Conversions|ToLinearColor(Color) (Variables|/Script/FortniteGame.BuildingTextureData|GetTintColor (Utilities|Array|Get(acopy) _outtexturedataarray))) _lightfixture)))

(fn BuildExposedAssetProviderVersion(Implementation)
  (params (Version :type Fort%20Exposed%20Asset%20Provider%20Version%20Structure%20%28by%20ref%29))
  (|CallParentFunction Version)
  (ExposedAssetProvider|AddVersion Version 2)
  (return))

(fn CollectAssetReferencestoAllow(Implementation)
  (params (Collector :type Fort%20Exposed%20Asset%20Reference%20Collector%20Structure%20%28by%20ref%29))
  (bind _mainmaterial (Variables|Default|GetMainMaterial))
  (|CallParentFunction Collector)
  (ExposedAssetProvider|CollectIndirectAssetReference Collector (Utilities|ResolveSoftReference _mainmaterial))
  (ExposedAssetProvider|CollectIndirectAssetReference Collector (Utilities|ResolveSoftReference (Rendering|Material|GetNaniteOverideMaterial _mainmaterial)))
  (ExposedAssetProvider|CollectObjectPropertyAssetReferences Collector 0 (Variables|Default|GetMainStaticMesh))
  (return))

(fn ConstructionScript
  (|CallParentFunction)
  (CallFunction|ApplyCorrectMIDFromOverride)
  (Rendering|SetVisibility (Variables|Debug|GetStaticMeshComponent) false false)
  (bind _returnvalue (AddComponent|NODE_AddStaticMeshComponent-0))
  (Components|StaticMesh|SetStaticMesh _returnvalue (Variables|/Script/Engine.StaticMeshComponent|GetStaticMesh (Variables|Debug|GetStaticMeshComponent))))
